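; Quest gameplay ability that reports damage dealt to tagged pawns while they are below a health threshold, apparently 25 percent.
; It listens for Damage.Dealt events, filters targets, and broadcasts the computed damage as a custom verb message.
; CalculateDamage infers pre-hit health to count only the portion under the threshold.
; For zero-health results, follow-up hits subtract the last reported damage from the threshold; first hits use a random placeholder.
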(event Ability|EventActivateAbility
  (AsyncAction|Cancel (Variables|Default|GetAsync%20Action))
  (bind (_asynctaskproxy _payload) (Ability|Async|WaitGameplayEventToActor (Ability|GetAvatarActorfromActorInfo) "(TagName=\"Event.Damage.Dealt\")")
    (:then
      (Variables|Default|SetAsync%20Action _asynctaskproxy))
    (:EventReceived
      (CallFunction|ClearInvalidTargets)
      (bind _success (CallFunction|CheckTags (.Target (Utilities|Struct|BreakGameplayEventData _payload))))
      (if _success
        (bind _validpawn (CallFunction|GatherTargetInfo (.EventMagnitude (Utilities|Struct|BreakGameplayEventData _payload)) (.Target (Utilities|Struct|BreakGameplayEventData _payload))))
        (if _validpawn
          (bind _success_2 (CallFunction|TargetIsBelowThreshold))
          (if _success_2
            (bind _finaldamage (CallFunction|CalculateDamage))
            (Variables|Default|SetFinalDamageDone _finaldamage)
            (CustomVerb|BroadcastCustomVerbMessage (CustomVerb|MakeCustomVerbMessage (Ability|GetOwningActorfromActorInfo) "(TagName=\"CustomVerb.Damage.Under25Percent\")" (Variables|Default|GetFinalDamageDone)))
            (Development|PrintString (Utilities|String|Append "Final Damage sent to Quest: " (Utilities|String|ToString(Integer) (Variables|Default|GetFinalDamageDone))) false false "(R=1.000000,G=0.100000,B=0.100000,A=1.000000)" 3.0)))))))

(event Ability|EventOnEndAbility
  (params (bWasCancelled :type Boolean))
  (AsyncAction|Cancel (Variables|Default|GetAsync%20Action))
  (CallFunction|ClearInvalidTargets))

(fn CalculateDamage
  (bind _damagedone (Variables|Default|GetDamageDone))
  (bind _damagedone_2 (Variables|Default|GetDamageDone))
  (if (Variables|Default|GetFollowupHit)
    (if (not (Math|Float|NearlyEqual(Float) (Variables|Default|GetTargetCurrentHealth) 0.0 0.1))
      (bind _output_get (Variables|Default|SetHealthBeforeDamage (+ (Variables|Default|GetTargetCurrentHealth) (Variables|Default|GetDamageDone))))
      (if (< (- _output_get _damagedone_2) 0)
        (Variables|Default|SetDamageDone (- (Variables|Default|GetDamageDone) (Math|Float|Absolute(Float) (- _output_get _damagedone_2)))))
      (if (> (Variables|Default|GetHealthBeforeDamage) (Variables|Default|GetTargetHealthThreshold))
        (bind _output_get_2 (Variables|Default|SetFinalDamageDone (Math|Float|Truncate (- (Variables|Default|GetDamageDone) (- (Variables|Default|GetHealthBeforeDamage) (Variables|Default|GetTargetHealthThreshold))))))
        (return _output_get_2)
        (else
          (bind _output_get_3 (Variables|Default|SetFinalDamageDone (Math|Float|Truncate (Variables|Default|GetDamageDone))))
          (return _output_get_3)))
      (else
        (bind _output_get_4 (Variables|Default|SetFinalDamageDone (Math|Float|Truncate (- (Variables|Default|GetTargetHealthThreshold) (Variables|Default|GetFinalDamageDone)))))
        (return _output_get_4)))
    (else
      (if (not (Math|Float|NearlyEqual(Float) (Variables|Default|GetTargetCurrentHealth) 0.0 0.1))
        (bind _output_get_5 (Variables|Default|SetHealthBeforeDamage (+ (Variables|Default|GetTargetCurrentHealth) (Variables|Default|GetDamageDone))))
        (if (< (- _output_get_5 _damagedone) 0)
          (Variables|Default|SetDamageDone (- (Variables|Default|GetDamageDone) (Math|Float|Absolute(Float) (- _output_get_5 _damagedone)))))
        (if (> (Variables|Default|GetHealthBeforeDamage) (Variables|Default|GetTargetHealthThreshold))
          (bind _output_get_6 (Variables|Default|SetFinalDamageDone (Math|Float|Truncate (- (Variables|Default|GetDamageDone) (- (Variables|Default|GetHealthBeforeDamage) (Variables|Default|GetTargetHealthThreshold))))))
          (return _output_get_6)
          (else
            (bind _output_get_7 (Variables|Default|SetFinalDamageDone (Math|Float|Truncate (Variables|Default|GetDamageDone))))
            (return _output_get_7)))
        (else
          ;[1 0 0.5] Temporary solution for hits with an unknown starting health value, technically player imperceptible
          (Development|PrintString "RANDOM VALUE CHOSEN!!!" false false "(R=1.000000,G=0.000000,B=0.000000,A=1.000000)" 1e+01)
          (bind _output_get_8 (Variables|Default|SetFinalDamageDone (Math|Float|Truncate (Math|Random|RandomFloatinRange 4.0 12.0))))
          (return _output_get_8))))))

(fn CheckTags
  (params (InTargetActor :type Actor%20Object%20Reference))
  (Utilities|IsValid InTargetActor
    (:Is%20Valid
      (bind _returnvalue (GameplayTags|HasAnyMatchingGameplayTags(Message) InTargetActor (Variables|Default|GetValidTargetTags)))
      (return _returnvalue))
    (:Is%20Not%20Valid
      (return false))))

(fn ClearInvalidTargets
  (for _array_element (Variables|Default|GetValidTargets)
    (Utilities|IsValid _array_element
      (:Is%20Not%20Valid
        (Utilities|Array|RemoveItem (Variables|Default|GetValidTargets) _array_element))))
  (return))

(fn CompareTarget
  (params (InTargetActor :type Actor%20Object%20Reference))
  (if (== InTargetActor (Variables|Default|GetCurrentTarget))
    (Variables|Default|SetFollowupHit true)
    (return)
    (else
      (Variables|Default|SetFollowupHit false)
      (Variables|Default|SetCurrentTarget (Variables|GetInTargetActor))
      (Utilities|Array|AddUnique (Variables|Default|GetValidTargets) (Variables|GetInTargetActor))
      (return))))

(fn GatherTargetInfo
  (params (InDamage :type Float%20%28double-precision%29) (InTargetActor :type Actor%20Object%20Reference))
  (Variables|Default|SetDamageDone InDamage)
  (CallFunction|CompareTarget InTargetActor)
  (bind _asfort_pawn (Utilities|Casting|CastToFortPawn InTargetActor)
    (:then
      (Variables|Default|SetTargetCurrentHealth (FortPawn|GetHealth _asfort_pawn))
      (bind _output_get (Variables|Default|SetTargetMaxHealth (FortPawn|GetMaxHealth _asfort_pawn)))
      (Variables|Default|SetTargetHealthThreshold (* _output_get (Variables|Default|GetGoalThreshold)))
      (Development|PrintString (Utilities|String|Append "Target Health: " (Utilities|String|ToString(Float) (FortPawn|GetHealth _asfort_pawn)) " | " "Target Max Health: " (Utilities|String|ToString(Float) (FortPawn|GetMaxHealth _asfort_pawn))) false false "(R=1.000000,G=0.100000,B=0.100000,A=1.000000)" 15.0)
      (return true))
    (:CastFailed
      (return false))))

(fn TargetIsBelowThreshold
  (return (<= (Variables|Default|GetTargetCurrentHealth) (Variables|Default|GetTargetHealthThreshold))))
